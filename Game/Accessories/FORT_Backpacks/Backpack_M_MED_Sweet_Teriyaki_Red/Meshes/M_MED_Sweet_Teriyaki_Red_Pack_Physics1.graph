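# Physics asset for the Sweet Teriyaki Red backpack mesh, giving simulation bodies to its body, eye, fin, tail and jiggle bones.
# Small capsules cover every part except the jiggle bone, which uses a sphere and is angularly locked to the body.
# Eyes, two-link fins and a two-link tail chain off the body with limited swing and very tight twist limits.
# The fin tip limits differ slightly between left and right.
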PhysicsAsset /Game/Accessories/FORT_Backpacks/Backpack_M_MED_Sweet_Teriyaki_Red/Meshes/M_MED_Sweet_Teriyaki_Red_Pack_Physics1
bodies {
 dyn_body [{shapeType=Capsule, length=5, radius=3, shapeName=dyn_body_capsule, rotation=<rot 90 0 0>}]
 dyn_eye_l [{shapeType=Capsule, length=2, radius=3, shapeName=dyn_eye_l_capsule, center=<2 0 0>, rotation=<rot 90 0 0>}]
 dyn_eye_r [{shapeType=Capsule, length=2, radius=3, shapeName=dyn_eye_l_capsule, center=<2 0 0>, rotation=<rot 90 0 0>}]
 dyn_fin_l1 [{shapeType=Capsule, length=2, radius=1, shapeName=dyn_fin_l1_capsule, center=<0.5 0 0>, rotation=<rot 0 0 90>}]
 dyn_fin_l2 [{shapeType=Capsule, length=2, radius=1, shapeName=dyn_fin_l1_capsule, center=<1 0 0>, rotation=<rot 0 0 90>}]
 dyn_fin_r1 [{shapeType=Capsule, length=2, radius=1, shapeName=dyn_fin_l1_capsule, center=<0.5 0 0>, rotation=<rot 0 0 90>}]
 dyn_fin_r2 [{shapeType=Capsule, length=2, radius=1, shapeName=dyn_fin_l1_capsule, center=<1 0 0>, rotation=<rot 0 0 90>}]
 dyn_jiggle [{shapeType=Sphere, radius=3, shapeName=dyn_jiggle_sphere}]
 dyn_tail_1 [{shapeType=Capsule, length=3.25, radius=1, shapeName=dyn_fin_l1_capsule, center=<0.5 0 0>}]
 dyn_tail_2 [{shapeType=Capsule, length=3.25, radius=1, shapeName=dyn_fin_l1_capsule, center=<1 0 0>}]
}
constraints [
 {
  bone1Name=dyn_jiggle
  bone2Name=dyn_body
  swing1LimitDegrees=45
  swing1Motion=Locked
  swing2LimitDegrees=45
  swing2Motion=Locked
  twistLimitDegrees=45
  twistMotion=Locked
 }
 {bone1Name=dyn_eye_l, bone2Name=dyn_body, swing1LimitDegrees=45, swing2LimitDegrees=45, twistLimitDegrees=1}
 {bone1Name=dyn_eye_r, bone2Name=dyn_body, swing1LimitDegrees=45, swing2LimitDegrees=45, twistLimitDegrees=1}
 {bone1Name=dyn_fin_l1, bone2Name=dyn_body, swing1LimitDegrees=6, swing2LimitDegrees=25, twistLimitDegrees=1}
 {bone1Name=dyn_fin_l2, bone2Name=dyn_fin_l1, swing1LimitDegrees=15, swing2LimitDegrees=30, twistLimitDegrees=1}
 {bone1Name=dyn_fin_r1, bone2Name=dyn_body, swing1LimitDegrees=6, swing2LimitDegrees=25, twistLimitDegrees=1}
 {bone1Name=dyn_fin_r2, bone2Name=dyn_fin_r1, swing1LimitDegrees=10, swing2LimitDegrees=30, twistLimitDegrees=1}
 {bone1Name=dyn_tail_1, bone2Name=dyn_body, swing1LimitDegrees=25, swing2LimitDegrees=5, twistLimitDegrees=1}
 {bone1Name=dyn_tail_2, bone2Name=dyn_tail_1, swing1LimitDegrees=25, swing2LimitDegrees=5, twistLimitDegrees=1}
]
